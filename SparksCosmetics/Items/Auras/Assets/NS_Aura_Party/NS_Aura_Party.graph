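# Niagara system for the Party cosmetic aura, showing small colored firework pops around a character's head alongside glows attached to skeletal mesh sockets or bones.
# FireworkPops spawns randomly with lights, feeding smoke and burst-mesh emitters through particle attribute readers, and each burst mesh sprays falling sparks.
# Glow and HeadGlow smoothly blend between seeded picks from the shared FireworkColor palette.
# Front-end and large-body flags adjust brightness, offsets and size.
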NiagaraSystem /SparksCosmetics/Items/Auras/Assets/NS_Aura_Party/NS_Aura_Party
dependencies {
 usedDataInterfaces [
  &NiagaraDataInterfaceParticleRead
  &NiagaraDataInterfaceCurve
  &NiagaraDataInterfaceColorCurve
  &NiagaraDataInterfaceArrayFloat
  &NiagaraDataInterfaceArrayInt32
  &NiagaraDataInterfaceSkeletalMesh
  &NiagaraDataInterfaceArrayColor
 ]
 usedDynamicInputs [&FN_FrontEndFloatModifier, &UniformRangedFloat]
 usedModules [
  &self:SystemSpawnScript.SystemScriptSource.SystemScriptGraph.NiagaraNodeAssignment_0.SetVariables_#1
  &self:SystemSpawnScript.SystemScriptSource.SystemScriptGraph.NiagaraNodeAssignment_1.SetVariables_#2
  &SystemState
 ]
}
emitters {
 FireworkPops {
  emitterUpdateScript [
   &EmitterState()
   &SpawnBurst_Instantaneous(Use Spawn Probability=false, UseLoopCountLimit=true)
   &SpawnRate(Spawn Probability=0.5, SpawnRate=2, Use Spawn Probability=true)
   &SpawnRate(Spawn Probability=0.75, SpawnRate=0.6, Use Spawn Probability=true) {module=SpawnRate001}
  ]
  particleSpawnScript [
   &InitializeParticle(
    Color Mode=::NewEnumerator0
    Lifetime=&Add_Float(A=0.2, B=$Engine.DeltaTime)
    Sprite Size Mode=::NewEnumerator1
    Uniform Sprite Size Max=140
    Uniform Sprite Size Min=120
    UsePositionOffset=false
    Write Material Random=true
   )
   /SetVariables(Particles.FireworkColor:LinearColor=&MultiplyLinearColor(
     Linear Color=&SelectLinearColorFromArray(Array Sampling Mode=::NewEnumerator0, Direct Array Index=?, Color Selection Array=$System.FireworkColor, Random Seed=0)
     Scale Factor=&MakeFloat4FromVectorAndFloat(
      "Vector (RGB)"=&VectorFromFloat(Value=&MakeCustomFloatFromBool(Bool=$User.IsFrontEnd, True Float=0.5, False Float=1))
      "Float (Alpha)"=1
     )
    )) {isSetParameters=true}
   /SetVariables(Particles.FireworkColor:LinearColor=&MultiplyLinearColor(
     Linear Color=&MakeCustomLinearColorFromBool(Bool=&RandomBool(), True Color=<rgba 0.2781 3 0.7437 1>, False Color=<rgba 3 0.3531 2.2213 1>)
     Scale Factor=&MakeFloat4FromVectorAndFloat(
      "Vector (RGB)"=&VectorFromFloat(Value=&MakeCustomFloatFromBool(Bool=$User.IsFrontEnd, True Float=0.5, False Float=1))
      "Float (Alpha)"=1
     )
    )) {enabled=false, isSetParameters=true}
   /SetVariables(Particles.LightExponent=1, Particles.LightRadius=30) {isSetParameters=true}
   &SkeletalMeshLocation(
    Bone / Socket Index=3
    Bone Sampling Mode=::NewEnumerator4
    MaskSpawnGroup=false
    Sampled Position Offset=&MakeCustomVectorFromBool(Bool=$User.IsFrontEnd, True Vector=<0 0 -40>, False Vector=<0 0 0>)
    Sampled Position Offset Coordinate Space:ENiagaraSkelMeshTransforms=::NewEnumerator1
    Skeletal Mesh=$System.Skeletal Mesh
   )
   &ShapeLocation(
    Box Midpoint=<0.5 0.5 0.25>
    Box Size=<100 100 75>
    MaskSpawnGroup=false
    Override Seed=false
    Shape Primitive=::NewEnumerator2
    Surface Only=false
    UseOwnerScale=true
   )
   /SetVariables(Particles.LightPos:NiagaraPosition=&Lerp_Vector(
     A=Linked() {linkedVariable {name=Particles.Position, type=NiagaraPosition}}
     B=$Output.SkeletalMeshLocation.SampledPositionWithoutOffset
     Alpha=0.5
    )) {isSetParameters=true}
   &CameraOffset(Camera Offset Amount=&MakeCustomFloatFromBool(Bool=$User.IsFrontEnd, True Float=-150, False Float=-50)) {module=CameraOffset001}
  ]
  particleUpdateScript [
   &ParticleState()
   &ScaleColor(
    Color Value To Scale=$Particles.Initial.FireworkColor
    Scale Alpha=&FloatFromCurve(FloatCurve=DataInterface() {propertyValues {Curve {keys [
         {arriveTangent=-0.1808, interpMode=RCIM_Cubic, leaveTangent=-0.1808, tangentMode=RCTM_User, time=0.4, value=0}
         {arriveTangent=4.059, interpMode=RCIM_Cubic, leaveTangent=4.059, tangentMode=RCTM_User, time=0.6, value=1}
         {arriveTangent=-2.488, interpMode=RCIM_Cubic, leaveTangent=-2.488, time=0.7, value=0.4}
         {interpMode=RCIM_Cubic, leaveTangent=-0.8972, time=1.0011, value=0.002}
        ]}}})
    Scale RGB=<5 5 5>
    ScaleA=true
    ScaleRGB=true
   )
   &ScaleColor(
    Color Value To Scale=$Particles.Initial.FireworkColor
    Scale Alpha=&FloatFromCurve(FloatCurve=DataInterface() {propertyValues {Curve {keys [
         {arriveTangent=-0.1808, interpMode=RCIM_Cubic, leaveTangent=-0.1808, tangentMode=RCTM_User, time=0, value=0}
         {arriveTangent=4.059, interpMode=RCIM_Cubic, leaveTangent=4.059, tangentMode=RCTM_User, time=0.1, value=1}
         {arriveTangent=-1.1075, interpMode=RCIM_Cubic, leaveTangent=-1.1075, time=0.2, value=0.4}
         {interpMode=RCIM_Cubic, leaveTangent=-0.8972, time=1.0011, value=0.002}
        ]}}})
    Scale RGB=<5 5 5>
    ScaleA=true
    ScaleRGB=true
   ) {enabled=false, module=ScaleColor001}
   &ScaleSpriteSize(Scale Factor=&Vector2DFromFloat(Value=&FloatFromCurve(FloatCurve=DataInterface() {propertyValues {Curve {keys [
          {arriveTangent=35.674, interpMode=RCIM_Cubic, leaveTangent=35.6741, tangentMode=RCTM_User, time=0, value=0}
          {arriveTangentWeight=0.0153, interpMode=RCIM_Cubic, leaveTangentWeight=0.3181, tangentMode=RCTM_User, time=0.0458, value=1.0493}
          {arriveTangent=-0.6879, interpMode=RCIM_Cubic, leaveTangent=-0.6879, tangentMode=RCTM_User, time=1, value=0.5}
         ]}}})))
   /SetVariables(Particles.LightColor=&MultiplyLinearColor(
     Linear Color=&ColorFromCurve(ColorCurve=DataInterface() {propertyValues {
        AlphaCurve {keys [<at 0 1 RCIM_Cubic>]}
        BlueCurve {keys [
          {arriveTangent=-22.8171, interpMode=RCIM_Cubic, leaveTangent=-22.8171, tangentMode=RCTM_User, time=0, value=5}
          {arriveTangent=-3.6507, interpMode=RCIM_Cubic, leaveTangent=-3.6507, tangentMode=RCTM_User, time=0.3667, value=2}
          {arriveTangent=-8.6995, interpMode=RCIM_Cubic, leaveTangent=-8.6995, tangentMode=RCTM_User, time=1, value=0}
         ]}
        GreenCurve {keys [
          {arriveTangent=-22.8171, interpMode=RCIM_Cubic, leaveTangent=-22.8171, tangentMode=RCTM_User, time=0, value=5}
          {arriveTangent=-3.6507, interpMode=RCIM_Cubic, leaveTangent=-3.6507, tangentMode=RCTM_User, time=0.3667, value=2}
          {arriveTangent=-8.6995, interpMode=RCIM_Cubic, leaveTangent=-8.6995, tangentMode=RCTM_User, time=1, value=0}
         ]}
        RedCurve {keys [
          {arriveTangent=-22.8171, interpMode=RCIM_Cubic, leaveTangent=-22.8171, tangentMode=RCTM_User, time=0, value=5}
          {arriveTangent=-3.6507, interpMode=RCIM_Cubic, leaveTangent=-3.6507, tangentMode=RCTM_User, time=0.3667, value=2}
          {arriveTangent=-8.6995, interpMode=RCIM_Cubic, leaveTangent=-8.6995, tangentMode=RCTM_User, time=1, value=0}
         ]}
       }})
     Scale Factor=&MakeVector4_FromLinearColor(LinearColor=$Particles.Initial.FireworkColor)
    )) {isSetParameters=true}
   /SetVariables(Particles.FireworkColor:LinearColor=&MakeLinearColorFromVectorAndFloat(
     "Vector (RGB)"=&MakeVectorFromLinearColorRGB(Color=&LerpLinearColors(
       StartColor=<rgba 1 1 1 1>
       EndColor=$Particles.Initial.FireworkColor
       LerpFactor=&FloatFromCurve(FloatCurve=DataInterface() {propertyValues {Curve {keys [
            {arriveTangent=2.324, interpMode=RCIM_Cubic, leaveTangent=2.324, tangentMode=RCTM_User, time=0, value=0}
            {tangentMode=RCTM_User, time=1, value=1}
           ]}}})
      ))
     "Float (Alpha)"=&MakeFloatFromLinearColor(LinearColor=$Particles.Color, Channel:ENiagaraLinearColor_Channels=::NewEnumerator3)
    )) {enabled=false, isSetParameters=true}
  ]
  rendererClasses [&NiagaraLightRendererProperties, &NiagaraSpriteRendererProperties]
  renderers [
   {rendererClass=&NiagaraLightRendererProperties, rendererIndex=0}
   {rendererClass=&NiagaraSpriteRendererProperties, rendererIndex=1}
   {rendererClass=&NiagaraSpriteRendererProperties, rendererIndex=2}
  ]
 }
 FireworkSmoke {
  emitterUpdateScript [
   &EmitterState()
   &SpawnParticlesFromOtherEmitter(
    Attribute Reader=DataInterface() {propertyValues {EmitterBinding {bindingMode=Other, emitterName=FireworkPops}}}
    Use Maximum Age=false
    Use Minimum Age=false
   )
  ]
  particleSpawnScript [
   &InitializeParticle(
    Color=<rgba 1 1 1 0.2>
    Lifetime Max=0.75
    Lifetime Min=0.5
    Lifetime Mode=::NewEnumerator1
    Sprite Rotation Mode=::NewEnumerator2
    Sprite Size Mode=::NewEnumerator1
    Uniform Sprite Size Max=125
    Uniform Sprite Size Min=100
    UsePositionOffset=false
    Write Material Random=true
   )
   &SampleParticlesFromOtherEmitter(
    Applied RGB Scale:Vector3f=&MakeVectorFromLinearColorRGB(Color=$Particles.Color)
    Apply Sampled Alpha=false
    Color Sampling=::NewEnumerator0
    Sprite Size Sampling:ENiagara_DirectReadSamplingMode=::NewEnumerator0
   )
   &CameraOffset(Camera Offset Amount=&MakeCustomFloatFromBool(Bool=$User.IsFrontEnd, True Float=-150, False Float=-50))
  ]
  particleUpdateScript [
   &ParticleState()
   &GravityForce(Gravity=<0 0 -35>)
   &SolveForcesAndVelocity(Clamp Velocity=false, Limit Acceleration=false)
   &SpriteRotationRate(Rotation Rate=15)
   &ScaleSpriteSize(Scale Factor=&Vector2DFromFloat(Value=&FloatFromCurve(
      FloatCurve=DataInterface() {propertyValues {Curve {keys [
          {arriveTangent=2.5076, interpMode=RCIM_Cubic, leaveTangent=2.5076, tangentMode=RCTM_User, time=0.0023, value=0}
          {arriveTangent=0.0897, interpMode=RCIM_Cubic, leaveTangent=0.0897, tangentMode=RCTM_User, time=1, value=1}
         ]}}}
      Scale Curve=1.2
     )))
   &ScaleColor(
    Scale Alpha=&FloatFromCurve(
     FloatCurve=DataInterface() {propertyValues {Curve {keys [
         {arriveTangent=20.383, interpMode=RCIM_Cubic, leaveTangent=20.383, tangentMode=RCTM_User, time=0, value=0}
         {arriveTangent=0.3312, interpMode=RCIM_Cubic, leaveTangent=-2.6189, tangentMode=RCTM_Break, time=0.1744, value=0.9882}
         {arriveTangent=-1.9376, interpMode=RCIM_Cubic, leaveTangent=-1.9376, tangentMode=RCTM_User, time=1, value=0}
        ]}}}
     Scale Curve=&MakeCustomFloatFromBool(Bool=$User.IsFrontEnd, True Float=0.3, False Float=0.5)
    )
    Scale RGB=&MakeVectorFromLinearColorRGB(Color=&ColorFromCurve(ColorCurve=DataInterface() {propertyValues {
        AlphaCurve {keys []}
        BlueCurve {keys [<at 0 1 RCIM_Cubic>,<at 1 0.459 RCIM_Cubic>]}
        GreenCurve {keys [<at 0 1 RCIM_Cubic>,<at 1 0.0509 RCIM_Cubic>]}
        RedCurve {keys [<at 0 1 RCIM_Cubic>,<at 1 0.255 RCIM_Cubic>]}
       }}))
    ScaleA=true
    ScaleRGB=true
   ) {enabled=false}
   /SetVariables(Particles.Color=&MakeLinearColorFromVectorAndFloat(
     "Vector (RGB)"=&Lerp_Vector(
      A=&MakeVectorFromLinearColorRGB(Color=$Particles.Initial.Color)
      B=<1 1 1>
      Alpha=$Particles.NormalizedAge
     )
     "Float (Alpha)"=&FloatFromCurve(
      FloatCurve=DataInterface() {propertyValues {Curve {keys [
          {arriveTangent=20.383, interpMode=RCIM_Cubic, leaveTangent=20.383, tangentMode=RCTM_User, time=0, value=0}
          {arriveTangent=0.3312, interpMode=RCIM_Cubic, leaveTangent=-2.6189, tangentMode=RCTM_Break, time=0.1744, value=0.9882}
          {arriveTangent=-1.9376, interpMode=RCIM_Cubic, leaveTangent=-1.9376, tangentMode=RCTM_User, time=1, value=0}
         ]}}}
      Scale Curve=&Multiply_Float(
       A=&MakeFloatFromLinearColor(LinearColor=$Particles.Initial.Color, Channel:ENiagaraLinearColor_Channels=::NewEnumerator3)
       B=&MakeCustomFloatFromBool(Bool=$User.IsFrontEnd, True Float=0.3, False Float=0.5)
      )
     )
    )) {isSetParameters=true}
   &self:ColorFromTOD_12(
    Afternoon:LinearColor=<rgba 1 0.9954 0.9249 1>
    Evening:LinearColor=<rgba 0.989 1 0.7298 1>
    Morning:LinearColor=<rgba 1 0.9526 0.8666 1>
    Night:LinearColor=<rgba 0.2955 0.5116 0.7 1>
    TOD:Vector4f=&MakeVector4_FromLinearColor(LinearColor=$NPC.FortniteNPC.FortniteActiveTimeOfDay)
   ) {module=ColorFromTOD_12}
  ]
  rendererClasses [&NiagaraSpriteRendererProperties]
  renderers [{rendererClass=&NiagaraSpriteRendererProperties}]
 }
 Firework_Bits {
  emitterUpdateScript [
   &EmitterState()
   &SpawnBurst_Instantaneous(
    Spawn Count=&RandomRangeInt(Minimum=150, Maximum=200)
    Spawn Time=0.05
    Use Spawn Probability=false
    UseLoopCountLimit=false
   ) {enabled=false}
   &SpawnParticlesFromOtherEmitter(
    Attribute Reader=DataInterface() {propertyValues {EmitterBinding {bindingMode=Other, emitterName=Fireworks_Mesh}}}
    Spawn Rate=1000
    Spawn Rate Per Particle Cap=1000
    Use Maximum Age=false
    Use Minimum Age=false
   )
  ]
  particleSpawnScript [
   /SetVariables(Particles.bIsRandomColor:NiagaraBool=&RandomBool()) {enabled=false, isSetParameters=true}
   &InitializeParticle(
    Color Mode=::NewEnumerator0
    Lifetime Max=&RandomRangeFloat(Minimum=0.4, Maximum=0.7)
    Lifetime Min=0.2
    Lifetime Mode=::NewEnumerator1
    Sprite Rotation Mode=::NewEnumerator2
    Sprite Size Mode=::NewEnumerator1
    Uniform Sprite Size Max=15
    Uniform Sprite Size Min=7
    UsePositionOffset=false
    Write Material Random=true
   )
   &SampleParticlesFromOtherEmitter(
    Age Sampling:ENiagara_DirectReadSamplingMode=::NewEnumerator1
    Apply Sampled Alpha=false
    Color Sampling=::NewEnumerator0
    Mesh Scale Sampling:ENiagara_DirectReadSamplingMode=::NewEnumerator1
    Position Sampling:ENiagara_DirectReadSamplingMode=::NewEnumerator1
   )
   /SetVariables(Particles.NormalizedParentScale:NiagaraFloat=&Add_Float(
     A=&Multiply_Float(
      A=&VectorLength(Vector=$Output.SampleParticlesFromOtherEmitter.SampledScale, Safe Length Calculation=false)
      B=50
     )
     B=-0.25
    )) {isSetParameters=true}
   &ShapeLocation(
    MaskSpawnGroup=false
    Override Seed=false
    Shape Origin=$Output.SampleParticlesFromOtherEmitter.SampledPosition
    Sphere Radius=1
    UseOwnerScale=true
   )
   &AddVelocity(
    Module.Use Falloff Distance=false
    Override Seed=false
    Velocity Mode=::NewEnumerator1
    Velocity Origin=$Output.SampleParticlesFromOtherEmitter.SampledPosition
    Velocity Speed=&Multiply_Float(
     A=&RandomRangeFloat(Minimum=0.5, Maximum=1.5)
     B=&Lerp_Float(A=85, B=150, Alpha=$Particles.NormalizedParentScale)
    )
   )
   &KillParticles(Kill Particles=&SetBoolByFloatComparison(A=$Output.SampleParticlesFromOtherEmitter.SampledAge, B=&Add_Float(A=0.05, B=$Engine.DeltaTime)))
   /SetVariables(Particles.Color=&MakeCustomLinearColorFromBool(
     Bool=&RandomBool()
     True Color=$Particles.Color
     False Color=&LerpLinearColors(
      StartColor=&SelectLinearColorFromArray(Array Sampling Mode=::NewEnumerator0, Direct Array Index=?, Color Selection Array=$System.FireworkColor, Random Seed=0)
      EndColor=$Particles.Color
      LerpFactor=0.5
     )
    )) {enabled=false, isSetParameters=true}
   /SetVariables(Particles.Color=&HueShiftLinearColor(
     Hue=&RandomRangeFloat(Minimum=-0.01, Maximum=0.01)
     Saturation=1
     Value=&RandomRangeFloat(Minimum=0.01, Maximum=0.2)
     Linear Color=$Particles.Color
    )) {enabled=false, isSetParameters=true}
   &CameraOffset(Camera Offset Amount=&MakeCustomFloatFromBool(Bool=$User.IsFrontEnd, True Float=-150, False Float=-100)) {module=CameraOffset001}
  ]
  particleUpdateScript [
   &ParticleState()
   &Drag(
    Drag=&Lerp_Float(A=2, B=4, Alpha=$Particles.NormalizedParentScale)
    Use Linear Drag=true
    Use Rotational Drag=false
   )
   &GravityForce(Gravity=&Multiply_VectorByFloat(
     Vector=&RandomRangeVector(Minimum=<0 0 -100>, Maximum=<0 0 -300>)
     Float=&Lerp_Float(A=0.5, B=1, Alpha=$Particles.NormalizedParentScale)
    ))
   &SolveForcesAndVelocity(Clamp Velocity=false, Limit Acceleration=false)
   &ScaleSpriteSizeBySpeed(Max Scale Factor=<0.01 20>, Velocity Threshold=500) {enabled=false}
   &ScaleColor(
    Scale Alpha=&FloatFromCurve(FloatCurve=DataInterface() {propertyValues {Curve {keys [{interpMode=RCIM_Cubic, tangentMode=RCTM_Break, time=0, value=6}, {interpMode=RCIM_Cubic, tangentMode=RCTM_Break, time=1, value=0}]}}})
    Scale RGB=&VectorFromFloat(Value=&RandomRangeFloat(Minimum=0.01, Maximum=0.5))
    ScaleA=true
    ScaleRGB=true
   )
   &ScaleSpriteSize(Scale Factor=&Vector2DFromFloat(Value=&FloatFromCurve(FloatCurve=DataInterface() {propertyValues {Curve {keys [
          {arriveTangent=4.8499, interpMode=RCIM_Cubic, leaveTangent=4.8499, tangentMode=RCTM_User, time=0, value=0}
          {arriveTangent=1.4365, interpMode=RCIM_Cubic, leaveTangent=1.4365, tangentMode=RCTM_User, time=0.2127, value=0.701}
          {arriveTangent=0.0623, interpMode=RCIM_Cubic, leaveTangent=0.0623, time=0.75, value=1}
          {arriveTangent=-1.2496, interpMode=RCIM_Cubic, leaveTangent=-1.2496, tangentMode=RCTM_User, time=1, value=0.75}
         ]}}})))
  ]
  rendererClasses [&NiagaraSpriteRendererProperties]
  renderers [{rendererClass=&NiagaraSpriteRendererProperties}]
 }
 Fireworks_Mesh {
  emitterUpdateScript [
   &EmitterState()
   &SpawnRate(SpawnRate=5, Use Spawn Probability=false) {enabled=false}
   &SpawnParticlesFromOtherEmitter(
    Attribute Reader=DataInterface() {propertyValues {EmitterBinding {bindingMode=Other, emitterName=FireworkPops}}}
    Spawn Rate=6
    Use Maximum Age=false
    Use Minimum Age=false
   )
  ]
  particleSpawnScript [
   /SetVariables(
    Particles.Density:NiagaraFloat=&RandomRangeFloat(Minimum=0, Maximum=1)
    Particles.Scalar:NiagaraFloat=&RandomRangeFloat(Minimum=0, Maximum=1)
   ) {isSetParameters=true}
   &InitializeParticle(
    Lifetime=&Lerp_Float(
     A=0.3
     B=0.8
     Alpha=&FloatFromCurve(
      FloatCurve=DataInterface() {propertyValues {Curve {keys [
          {arriveTangent=0.9669, interpMode=RCIM_Cubic, leaveTangent=0.9669, tangentMode=RCTM_User, time=0, value=0}
          {arriveTangent=0.5716, interpMode=RCIM_Cubic, leaveTangent=0.5716, tangentMode=RCTM_User, time=0.75, value=0.4}
          {arriveTangent=4.5163, tangentMode=RCTM_User, time=1, value=1}
         ]}}}
      CurveIndex=&RandomRangeFloat(Minimum=0, Maximum=1)
     )
    )
    Mesh Scale Mode=::NewEnumerator3
    Mesh Uniform Scale=&Lerp_Float(A=0.01, B=0.03, Alpha=$Particles.Scalar)
    Sprite Size Mode=::NewEnumerator3
    Uniform Sprite Size=&Lerp_Float(A=90, B=130, Alpha=$Particles.Density)
    UsePositionOffset=false
    Write Material Random=true
   )
   &SampleParticlesFromOtherEmitter(Apply Sampled Alpha=false, Color Sampling=::NewEnumerator0)
   &InitialMeshOrientation(Use Rotation Vector=false)
   &CameraOffset(Camera Offset Amount=&MakeCustomFloatFromBool(Bool=$User.IsFrontEnd, True Float=-150, False Float=-100))
   &KillParticles(Kill Particles=&SetBoolByIntComparison(A=$Emitter.SpawnParticlesFromOtherEmitter.TotalSpawnedParticles, B=1, Comparison Type=::NewEnumerator0)) {enabled=false}
  ]
  particleUpdateScript [
   &ParticleState()
   &ScaleColor(
    Scale Alpha=&UniformAOrBFloat(
     A=&FloatFromCurve(FloatCurve=DataInterface() {propertyValues {Curve {keys [
          {arriveTangent=5.0329, interpMode=RCIM_Cubic, leaveTangent=5.0329, tangentMode=RCTM_User, time=0, value=1}
          {arriveTangentWeight=0.0464, interpMode=RCIM_Cubic, leaveTangentWeight=0.2879, tangentMode=RCTM_User, time=0.1389, value=1.3591}
          {arriveTangent=-2.0042, interpMode=RCIM_Cubic, leaveTangent=-2.0042, tangentMode=RCTM_User, time=1, value=0.4}
         ]}}})
     B=&FloatFromCurve(
      FloatCurve=DataInterface() {propertyValues {Curve {keys [
          {arriveTangent=-4.9376, interpMode=RCIM_Cubic, leaveTangent=7.1838, tangentMode=RCTM_User, time=0, value=1}
          {arriveTangent=-2.058, interpMode=RCIM_Cubic, leaveTangent=-2.058, tangentMode=RCTM_User, time=0.1389, value=1.3591}
          {arriveTangent=-0.3427, interpMode=RCIM_Cubic, leaveTangent=-0.3427, tangentMode=RCTM_User, time=1, value=0.25}
         ]}}}
      Scale Curve=2
     )
     Evaluation Type=::NewEnumerator0
     Recalculate Random Each Loop=true
     Override Seed=false
    )
    Scale RGB=&VectorFromFloat(Value=1)
    ScaleA=true
    ScaleRGB=false
   )
   &ScaleMeshSize(Scale Factor=&VectorFromFloat(Value=&FloatFromCurve(FloatCurve=DataInterface() {propertyValues {Curve {keys [
          {arriveTangent=3.3928, interpMode=RCIM_Cubic, leaveTangent=3.3928, tangentMode=RCTM_User, time=0, value=0}
          {arriveTangent=0.8374, interpMode=RCIM_Cubic, leaveTangent=0.8374, tangentMode=RCTM_User, time=0.3, value=0.5292}
          {arriveTangent=0.6991, interpMode=RCIM_Cubic, leaveTangent=0.6991, tangentMode=RCTM_User, time=1, value=1}
         ]}}})))
   &ScaleSpriteSize(Uniform Curve Sprite Scale=DataInterface() {propertyValues {Curve {keys [
        {arriveTangent=2.9523, interpMode=RCIM_Cubic, leaveTangent=2.9523, tangentMode=RCTM_User, time=0, value=0}
        {arriveTangent=1.6276, interpMode=RCIM_Cubic, leaveTangent=1.6276, tangentMode=RCTM_User, time=0.1923, value=0.7}
        {arriveTangent=0.2011, interpMode=RCIM_Cubic, leaveTangent=0.5033, tangentMode=RCTM_User, time=1, value=1}
       ]}}})
   &GravityForce(Gravity=<0 0 -30>)
   &SolveForcesAndVelocity(Clamp Velocity=false, Limit Acceleration=false)
   &DynamicMaterialParameters(
    Index 0 Param 1=1
    Index 0 Param 2=1
    Index 0 Param 3=&FloatFromCurve(
     FloatCurve=DataInterface() {propertyValues {Curve {keys [
         {arriveTangent=3.0114, interpMode=RCIM_Cubic, leaveTangent=3.0114, tangentMode=RCTM_User, time=0, value=0}
         {arriveTangent=0.9456, interpMode=RCIM_Cubic, leaveTangent=0.9456, tangentMode=RCTM_User, time=0.4343, value=0.7445}
         {arriveTangent=0.3236, interpMode=RCIM_Cubic, leaveTangent=0.7406, tangentMode=RCTM_User, time=1, value=1}
        ]}}}
     Scale Curve=&Lerp_Float(A=0.2, B=0.3, Alpha=$Particles.Density)
    )
    Index 0 Param 4=&FloatFromCurve(
     FloatCurve=DataInterface() {propertyValues {Curve {keys [
         {arriveTangent=3.4321, interpMode=RCIM_Cubic, leaveTangent=3.4321, tangentMode=RCTM_User, time=0, value=1}
         {arriveTangent=-1.7957, interpMode=RCIM_Cubic, leaveTangent=-1.7957, tangentMode=RCTM_User, time=0.1672, value=1.5}
         {arriveTangent=-0.286, interpMode=RCIM_Cubic, leaveTangent=-0.286, tangentMode=RCTM_User, time=1, value=0.5}
        ]}}}
     Scale Curve=0.5
    )
    Index 1 Param 1=2
    Module.Index0Param0Write=true
    Module.Index0Param1Write=true
    Module.Index0Param2Write=true
    Module.Index0Param3Write=true
    Module.Index1Param0Write=true
    Module.Index1Param1Write=true
    Module.Index1Param2Write=true
    Module.Index1Param3Write=true
    Write Parameter Index 1=true
   )
   /SetVariables(Particles.GlowColor=&MultiplyLinearColor(
     Linear Color=$Particles.Color
     Scale Factor=&MakeFloat4FromVectorAndFloat(
      "Vector (RGB)"=&VectorFromFloat(Value=0.333)
      "Float (Alpha)"=&FloatFromCurve(
       FloatCurve=DataInterface() {propertyValues {Curve {keys [<at 0 1 RCIM_Cubic>, {arriveTangent=-2.5514, tangentMode=RCTM_User, time=1, value=0}]}}}
       Scale Curve=0.08
      )
     )
    )) {isSetParameters=true}
  ]
  rendererClasses [&NiagaraMeshRendererProperties, &NiagaraSpriteRendererProperties]
  renderers [
   {rendererClass=&NiagaraMeshRendererProperties, rendererIndex=0}
   {rendererClass=&NiagaraSpriteRendererProperties, rendererIndex=1}
  ]
 }
 Glow {
  emitterUpdateScript [
   &EmitterState(Life Cycle Mode=::NewEnumerator1, Loop Duration=0.5, UseLoopDelay=false)
   &SpawnBurst_Instantaneous(Spawn Count=6, Use Spawn Probability=false, UseLoopCountLimit=true)
  ]
  particleSpawnScript [
   &PartitionParticles(Number Of Partitions=7)
   &InitializeParticle(
    Lifetime=0.1
    Sprite Rotation Mode=::NewEnumerator0
    Sprite Size Mode=::NewEnumerator3
    Uniform Sprite Size=&Multiply_Float(
     A=&SelectFloatFromArray(
      Direct Array Index=&SpawnGroup()
      Float Selection Array=DataInterface() {propertyValues {FloatData [40,40,20,20,30,30]}}
     )
     B=&MakeCustomFloatFromBool(Bool=$User.IsLargeBody, True Float=3, False Float=1)
    )
    UsePositionOffset=false
    Write Material Random=true
   )
   /SetVariables(Particles.glowScale:Vector2f=&Vector2DFromFloat(Value=&Multiply_Float(
      A=&SelectFloatFromArray(
       Direct Array Index=$Particles.PartitionParticles.Partition
       Float Selection Array=DataInterface() {propertyValues {FloatData [70,80,100,90,80,70,70,90]}}
      )
      B=&MakeCustomFloatFromBool(Bool=$User.IsLargeBody, True Float=1.2, False Float=1)
     ))) {isSetParameters=true}
   /SetVariables(Particles.glowCameraOffset=-10) {isSetParameters=true}
   /SetVariables(Particles.glowColor:LinearColor=&MultiplyLinearColor(
     Linear Color=&MultiplyLinearColor(Linear Color=<rgba 1 0.0427 0.8416 1>, Scale Factor=&MakeVector4_FromFloat(Float=3))
     Scale Factor=&MakeVector4_FromFloat(Float=&FNTimeOfDayScalar(
       Night Scale=0.25
       Morning Scale=0.2
       Day Scale=&MakeCustomFloatFromBool(Bool=$User.IsFrontEnd, True Float=0.07, False Float=0.26)
       Evening Scale=0.15
      ))
    )) {enabled=false, isSetParameters=true}
   /SetVariables(Particles.glowColor:LinearColor=&MultiplyLinearColor(
     Linear Color=&SelectLinearColorFromArray(
      Array Sampling Mode=::NewEnumerator0
      Direct Array Index=?
      Color Selection Array=$System.FireworkColor
      Randomness Mode=::NewEnumerator1
      Random Seed=&Add_Integer(A=$Emitter.LoopCount, B=$Particles.PartitionParticles.Partition)
      Fixed Random Seed=true
     )
     Scale Factor=&MakeVector4_FromFloat(Float=&MakeCustomFloatFromBool(Bool=$User.IsFrontEnd, True Float=0.5, False Float=1))
    )) {isSetParameters=true}
  ]
  particleUpdateScript [
   &ParticleState(
    Kill Particles When Lifetime Has Elapsed=false
    Let Infinitely Lived Particles Die When Emitter Deactivates=true
    Loop Particles Lifetime=true
   )
   /SetVariables(Particles.glowColor:LinearColor=/Niagara/DynamicInputs/Lerp/SmoothLerpOverTime_Color(
     Smooth Value:LinearColor=$Particles.glowColor
     Target Value:LinearColor=&MultiplyLinearColor(
      Linear Color=&SelectLinearColorFromArray(
       Array Sampling Mode=::NewEnumerator0
       Direct Array Index=?
       Color Selection Array=$System.FireworkColor
       Randomness Mode=::NewEnumerator1
       Random Seed=&Add_Integer(A=$Emitter.LoopCount, B=$Particles.PartitionParticles.Partition)
       Fixed Random Seed=true
      )
      Scale Factor=&MakeVector4_FromFloat(Float=&MakeCustomFloatFromBool(Bool=$User.IsFrontEnd, True Float=0.5, False Float=1))
     )
     Convergence Rate:NiagaraFloat=3
    )) {isSetParameters=true}
   /SetVariables(Particles.decalColor:LinearColor=&MultiplyLinearColor(Linear Color=$Particles.glowColor, Scale Factor=&MakeVector4_FromFloat(Float=2))) {isSetParameters=true}
   &CameraOffset(Camera Offset Amount=&MakeCustomFloatFromBool(Bool=$User.IsLargeBody, True Float=0, False Float=5))
   &SpriteRotationRate(Rotation Rate=&RandomRangeFloat(Minimum=10, Maximum=20))
   &SkeletalMeshLocation(
    Bone / Socket Index=&SelectIntFromArray(Direct Array Index=$Particles.PartitionParticles.Partition, Int Selection Array=DataInterface() {propertyValues {IntData [0,1,2,4,5,6]}})
    Bone Sampling Mode=::NewEnumerator4
    MaskSpawnGroup=false
    Orientation Sampling=::NewEnumerator1
    Sampled Position Offset Coordinate Space:ENiagaraSkelMeshTransforms=::NewEnumerator1
    Skeletal Mesh=$System.Skeletal Mesh
   ) {module=SkeletalMeshLocation003}
   &DynamicMaterialParameters(
    Index 0 Param 1=&SelectFloatFromArray(Direct Array Index=$Particles.PartitionParticles.Partition, Float Selection Array=DataInterface() {propertyValues {FloatData [5,5,5,5,3,3]}})
    Index 0 Param 3=1
    Module.Param0WriteEnabled=true
    Module.Param1WriteEnabled=true
    Module.Param2WriteEnabled=true
    Module.Param3WriteEnabled=true
   )
  ]
  rendererClasses [&NiagaraSpriteRendererProperties]
  renderers [
   {rendererClass=&NiagaraSpriteRendererProperties}
   {rendererClass=&NiagaraSpriteRendererProperties, rendererIndex=1}
  ]
 }
 HeadGlow {
  emitterUpdateScript [
   &EmitterState()
   &SpawnBurst_Instantaneous(Spawn Count=2, Use Spawn Probability=false, UseLoopCountLimit=true)
  ]
  particleSpawnScript [
   &InitializeParticle(
    Color=&MultiplyLinearColor(
     Linear Color=&SelectLinearColorFromArray(
      Array Sampling Mode=::NewEnumerator0
      Direct Array Index=?
      Color Selection Array=$System.FireworkColor
      Randomness Mode=::NewEnumerator1
      Random Seed=&Add_Integer(A=$Emitter.LoopCount, B=&ReturnExecIndex())
      Fixed Random Seed=true
     )
     Scale Factor=&MakeVector4_FromFloat(Float=&MakeCustomFloatFromBool(Bool=$User.IsFrontEnd, True Float=0.5, False Float=1))
    )
    Lifetime=0.1
    Sprite Size Mode=::NewEnumerator3
    Uniform Sprite Size=&RandomRangeFloat(Minimum=130, Maximum=170)
    UsePositionOffset=false
    Write Material Random=true
   )
   &CameraOffset(Camera Offset Amount=-5)
  ]
  particleUpdateScript [
   &ParticleState(
    Kill Particles When Lifetime Has Elapsed=false
    Let Infinitely Lived Particles Die When Emitter Deactivates=true
    Loop Particles Lifetime=true
   )
   &SkeletalMeshLocation(
    Bone / Socket Index=3
    Bone Sampling Mode=::NewEnumerator4
    MaskSpawnGroup=false
    Sampled Position Offset=&MakeCustomVectorFromBool(Bool=$User.IsLargeBody, True Vector=<5 0 1>, False Vector=<0 0 0>)
    Skeletal Mesh=$System.Skeletal Mesh
   )
   &ScaleSpriteSize(Uniform Curve Sprite Scale=DataInterface() {propertyValues {Curve {keys [
        {interpMode=RCIM_Cubic, tangentMode=RCTM_Break, time=0, value=0.5}
        {interpMode=RCIM_Cubic, tangentMode=RCTM_Break, time=0.5, value=1}
        {interpMode=RCIM_Cubic, tangentMode=RCTM_Break, time=1, value=0.5}
       ]}}}) {enabled=false}
   &SpriteRotationRate(Rotation Rate=&MakeCustomFloatFromBool(
     Bool=&SetBoolByIntComparison(A=&ReturnParticleIDAttributes(), B=0, Comparison Type=::NewEnumerator0)
     True Float=&RandomRangeFloat(Minimum=10, Maximum=15)
     False Float=&RandomRangeFloat(Minimum=-10, Maximum=-15)
    ))
   &Color(Color=&SmoothLerpOverTime_Color(
     Smooth Value:LinearColor=$Particles.Color
     Target Value:LinearColor=&MultiplyLinearColor(
      Linear Color=&SelectLinearColorFromArray(
       Array Sampling Mode=::NewEnumerator0
       Direct Array Index=?
       Color Selection Array=$System.FireworkColor
       Randomness Mode=::NewEnumerator1
       Random Seed=&Add_Integer(
        A=$Emitter.LoopCount
        B=&MakeIntFromFloat(Float=&ReturnNormalizedExecIndex(Normalized Index Scale=3))
       )
       Fixed Random Seed=true
      )
      Scale Factor=&MakeVector4_FromFloat(Float=&MakeCustomFloatFromBool(Bool=$User.IsFrontEnd, True Float=0.5, False Float=1))
     )
     Convergence Rate:NiagaraFloat=1
    ))
  ]
  rendererClasses [&NiagaraSpriteRendererProperties]
  renderers [{rendererClass=&NiagaraSpriteRendererProperties}]
 }
 Minimal {
  bEnabled=false
  emitterUpdateScript [
   &EmitterState()
   &SpawnBurst_Instantaneous(Spawn Count=9, Use Spawn Probability=false, UseLoopCountLimit=true)
  ]
  particleSpawnScript [&InitializeParticle(
    Color=&SelectLinearColorFromArray(
     Array Sampling Mode=::NewEnumerator1
     Direct Array Index=&MakeIntFromFloat(Float=&ReturnNormalizedExecIndex(Normalized Index Scale=9))
     Color Selection Array=$System.FireworkColor
     Random Seed=?
    )
    Lifetime=5
    Position Offset=&Lerp_Vector(A=<0 40 10>, B=<0 40 150>, Alpha=&ReturnNormalizedExecIndex())
    Sprite Size Mode=::NewEnumerator3
    UsePositionOffset=true
    Write Material Random=true
   )]
  particleUpdateScript [&ParticleState(Kill Particles When Lifetime Has Elapsed=false)]
  rendererClasses [&NiagaraSpriteRendererProperties]
  renderers [{rendererClass=&NiagaraSpriteRendererProperties}]
 }
}
summary {
 systemName=NS_Aura_Party
 emitters [
  {emitterName=Fireworks_Mesh, rendererClasses [&NiagaraMeshRendererProperties, &NiagaraSpriteRendererProperties]}
  {emitterName=FireworkPops, rendererClasses [&NiagaraLightRendererProperties, &NiagaraSpriteRendererProperties]}
  {emitterName=FireworkSmoke, rendererClasses [&NiagaraSpriteRendererProperties]}
  {emitterName=Glow, rendererClasses [&NiagaraSpriteRendererProperties]}
  {emitterName=HeadGlow, rendererClasses [&NiagaraSpriteRendererProperties]}
  {bEnabled=false, emitterName=Minimal, rendererClasses [&NiagaraSpriteRendererProperties]}
  {emitterName=Firework_Bits, rendererClasses [&NiagaraSpriteRendererProperties]}
 ]
 userVariables [
  User.bIsFrontEnd {type=NiagaraBool}
  User.bIsFrontEndPreview {type=NiagaraBool}
  User.BottomPosition {type=NiagaraPosition, description="", defaultValue {}}
  User.BottomSmokeDensity {type=NiagaraFloat, defaultValue=0, description=""}
  User.ColorVariant {type=NiagaraBool}
  User.FallWidth {type=NiagaraFloat, defaultValue=900, description=""}
  User.IsFrontEnd {type=NiagaraBool}
  User.IsLargeBody {type=NiagaraBool}
  User.SystemOffset {type=Vector3f, description="", defaultValue {struct=Vector3f, value=<0 0 125>}}
 ]
}
